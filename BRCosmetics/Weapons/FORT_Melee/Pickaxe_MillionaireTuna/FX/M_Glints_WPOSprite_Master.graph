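# Translucent unlit master material for flickering glint sprites on the MillionaireTuna pickaxe, flagged for Niagara mesh particles.
# World position offset billboards and camera-offsets quads, with vertex red varying size and vertex green matched against a dynamic Glint value to show them.
# Multiplied sines of time, offset by vertex red, modulate a radial mask that samples a corona gradient.
# That corona, a hot center and a tinted cross flare form emissive; the flare also drives opacity.
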Material /BRCosmetics/Weapons/FORT_Melee/Pickaxe_MillionaireTuna/FX/M_Glints_WPOSprite_Master
graph {
 properties {
  bUsedWithNiagaraMeshParticles=true
  blendMode=BLEND_Translucent
  pixelDepthOffsetMode=PDOM_Legacy
  shadingModel=MSM_Unlit
  twoSided=true
  usedShadingModels=Unlit
 }
 nodes [
  n1 = WorldPosition()
  n2 = TextureCoordinate()
  n3 = DynamicParameter() {parameterIndex=1, paramNames [Glint,Size,Param3,Param4]}
  n4 = VertexColor()
  n5 = Multiply(
   A=LinearInterpolate(A=Multiply(B=@n3.Size) {constA=0.5}, B=@n3.Size, Alpha=@n4.R)
   B=SphereMask(A=@n4.G, B=@n3.Glint, Radius=ScalarParameter() {defaultValue=0.01}) {hardnessPercent=1}
  )
  n6 = ScalarParameter() {defaultValue=8, parameterName=Camera Offset}
  NamedRerouteDeclaration(None=Add(
    A=MaterialFunctionCall!(Center Location=@n1.XYZ, 0-1 UVs=@n2, XY scale=@n5, OverridePivot, "Normals (Optional)") {materialFunction=&Sprite}.World position offset
    B=MaterialFunctionCall(Offset Amount=@n6) {materialFunction=/Engine/Functions/Engine_MaterialFunctions02/WorldPositionOffset/CameraOffset}.Clamped Camera Offset
   )) {name=WPO, nodeColor=<rgba 1 0 0.2344 1>}
  n7 = Constant() {r=0.5}
  n8 = SphereMask(A=TextureCoordinate(), B=@n7, Radius=@n7) {hardnessPercent=0}
  n9 = Multiply(A=Power(Base=@n8) {constExponent=25}, B=ScalarParameter() {defaultValue=2, parameterName=CenterBrightness}) {constB=2}
  n10 = Constant() {r=0.5}
  n11 = ParticleRandom()
  n12 = Multiply(
   A=Add(A=VertexColor().R, B=Time())
   B=StaticSwitchParameter(True=LinearInterpolate(Alpha=@n11) {constA=1.2, constB=0.8}, False=ScalarParameter() {defaultValue=1, parameterName=FlickerIntensity}) {parameterName=GPU Particles}
  )
  n13 = Sine(None=@n12) {period=0.155}
  n14 = Sine(None=@n12) {period=0.2}
  n15 = Multiply(A=Sine(None=@n12) {period=0.075}, B=Multiply(A=Multiply(A=@n13, B=@n14), B=Sine(None=@n12) {period=0.43}))
  n16 = LinearInterpolate(
   A=ScalarParameter() {defaultValue=0.4, desc="0 = full flicker amount", parameterName=GPUFlickerFadeMin}
   B=ScalarParameter() {defaultValue=1, desc="1 = no flicker", parameterName=GPUFlickerFadeMax, sortPriority=1}
   Alpha=ParticleRandom()
  ) {constA=0.5}
  n17 = DynamicParameter() {paramNames [FlickerFade,CoronaSize,Param3,Size]}
  n18 = Multiply(A=ScalarParameter() {defaultValue=1, parameterName=FlickerLerp}, B=@n17.FlickerFade)
  n19 = LinearInterpolate(Alpha=LinearInterpolate(A=Clamp(None=@n15) {maxDefault=3}, Alpha=StaticSwitchParameter(True=@n16, False=@n18) {parameterName=GPU Particles}) {constA=1}) {constA=-0.4, constB=0}
  n20 = ScalarParameter() {defaultValue=1, parameterName=CoronaSize}
  n21 = StaticSwitchParameter(True=@n20, False=Multiply(A=@n17.CoronaSize, B=@n20)) {parameterName=GPU Particles}
  n22 = SphereMask(A=TextureCoordinate(), B=@n10, Radius=@n10, Hardness=LinearInterpolate(Alpha=Add(A=@n19, B=@n21) {constB=0}) {constA=0.7, constB=-0.3}) {hardnessPercent=25}
  n23 = Multiply(
   A=Desaturation(
    None=TextureSampleParameter2D(UVs=@n22) {parameterName=ColorGradient, texture=/Game/Effects/Fort_Effects/Textures/Grad/T_Halo_Corona_Gradient}.RGB
    Fraction=ScalarParameter() {parameterName=Desaturate}
   )
   B=ScalarParameter() {defaultValue=0.125, parameterName=CoronaBrightness}
  ) {constB=0.125}
  n24 = VectorParameter() {parameterName=Color, defaultValue=<rgba 1 1 1 1>}
  n25 = ParticleColor()
  n26 = StaticSwitchParameter(True=@n24.RGB, False=Multiply(A=@n25.RGB, B=@n24.RGB)) {parameterName=UsedWithSkeletalMesh}
  n27 = Multiply(A=ShadingPathSwitch(Default=Add(A=@n9, B=@n23), Forward=@n23), B=LinearInterpolate(B=@n26) {constA=1, constAlpha=0.7})
  n28 = Power(Base=@n8) {constExponent=4}
  n29 = ScalarParameter() {defaultValue=0.15, parameterName=EdgeBrightness}
  n30 = Multiply(
   A=Add(
    A=Multiply(A=@n28, B=@n29) {constB=0.15}
    B=TextureSampleParameter2D() {parameterName=FlareMask, samplerType=SAMPLERTYPE_Grayscale, texture=/Game/Effects/Fort_Effects/Textures/Flares/T_Cross_Flare}.RGB
   )
   B=@n26
  )
  NamedRerouteDeclaration(None=MaterialFunctionCall(Emissive=Add(A=@n27, B=@n30)) {materialFunction=&MF_FX_EmissiveAdjustments}.Result) {name=Emissive, nodeColor=<rgba 1 0.9249 0.9671 1>}
  NamedRerouteDeclaration(None=Multiply(A=@n25.A, B=@n30)) {name=Opacity, nodeColor=<rgba 1 0 0.6797 1>}
  NamedRerouteUsage()
  NamedRerouteUsage()
  NamedRerouteUsage()
  ScalarParameter() {defaultValue=40, parameterName=XY}
 ]
}
