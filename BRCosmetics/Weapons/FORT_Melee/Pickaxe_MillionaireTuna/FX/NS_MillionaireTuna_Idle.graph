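# Niagara system for the idle effect on the MillionaireTuna pickaxe, drawing short-lived gold glints as mesh particles.
# A single TunaGlints emitter spawns at a set rate and, in particle update, samples the skeletal mesh filtered to the idle_fx socket for location.
# A color curve fades each glint in to a brightened peak and out.
# Dynamic material parameters feed the material random values and a randomly scaled rise-and-fall curve.
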NiagaraSystem /BRCosmetics/Weapons/FORT_Melee/Pickaxe_MillionaireTuna/FX/NS_MillionaireTuna_Idle
dependencies {
 usedDataInterfaces [&NiagaraDataInterfaceSkeletalMesh, &NiagaraDataInterfaceCurve, &NiagaraDataInterfaceColorCurve]
 usedDynamicInputs []
 usedModules [&SystemState]
}
emitters {TunaGlints {
  emitterUpdateScript [
   &EmitterState()
   &SpawnRate(Use Spawn Probability=false)
  ]
  particleSpawnScript [&InitializeParticle(
    Color=<rgba 1 0.4684 0 0.5>
    Lifetime Max=0.7
    Lifetime Min=0.35
    Lifetime Mode=::NewEnumerator1
    Mesh Scale Mode=::NewEnumerator3
    UsePositionOffset=false
    Write Material Random=true
   )]
  particleUpdateScript [
   &ParticleState()
   &SkeletalMeshLocation(
    MaskSpawnGroup=false
    Mesh Sampling Type=::NewEnumerator5
    Skeletal Mesh=DataInterface() {propertyValues {
      DefaultMesh=null
      ExcludeBoneName=root
      OptionalComponentClass=null
      PreviewMesh=/BRCosmetics/Weapons/FORT_Melee/Pickaxe_MillionaireTuna/Meshes/MillionaireTuna_Axe
      SkinningMode=SkinOnTheFly
      SoftSourceActor=null
      SourceMode=Default
      UvSetIndex=0
      WholeMeshLOD=-1
      bExcludeBone=false
      bReadDeformedGeometry=true
      bRequireCurrentFrameData=true
      ComponentTags []
      FilteredBones []
      FilteredSockets [idle_fx]
      MeshUserParameter {parameter {name=None, typeDefHandle {registeredTypeIndex=1}, varData []}}
      SamplingRegions []
     }}
    Socket Sampling Mode=::NewEnumerator5
   )
   &DynamicMaterialParameters(
    Index 0 Param 1=1
    Index 0 Param 2=1
    Index 1 Param 1=&RandomRangeFloat(Minimum=0, Maximum=1)
    Index 1 Param 2=&FloatFromCurve(
     FloatCurve=DataInterface() {propertyValues {Curve {keys [
         {interpMode=RCIM_Cubic, leaveTangent=2.2717, tangentMode=RCTM_Break, time=0, value=0}
         {interpMode=RCIM_Cubic, leaveTangent=-0.8734, tangentMode=RCTM_Break, time=0.5, value=1.05}
         {arriveTangent=-2.9536, interpMode=RCIM_Cubic, leaveTangent=-0.8734, tangentMode=RCTM_Break, time=1, value=0}
        ]}}}
     Scale Curve=&RandomRangeFloat(Minimum=50, Maximum=80)
    )
    Module.Index0Param0Write=true
    Module.Index0Param1Write=true
    Module.Index0Param2Write=true
    Module.Index0Param3Write=true
    Module.Index1Param0Write=true
    Module.Index1Param1Write=true
    Module.Index1Param2Write=true
    Module.Index1Param3Write=true
    Write Parameter Index 1=true
   )
   &ScaleColor(
    Linear Color Curve=DataInterface() {propertyValues {
      AlphaCurve {keys [<at 0.002 0>,<at 0.4078 1>,<at 1 0>]}
      BlueCurve {keys [<at 0.0008 0>,<at 0.4094 2>,<at 1 0>]}
      GreenCurve {keys [<at 0.0008 0>,<at 0.4094 2>,<at 1 0>]}
      RedCurve {keys [<at 0.0008 0>,<at 0.4094 2>,<at 1 0>]}
     }}
    Scale Mode=::NewEnumerator2
   )
  ]
  rendererClasses [&NiagaraMeshRendererProperties]
  renderers [{rendererClass=&NiagaraMeshRendererProperties, rendererIndex=0}]
 }}
summary {
 systemName=NS_MillionaireTuna_Idle
 emitters [{emitterName=TunaGlints, rendererClasses [&NiagaraMeshRendererProperties]}]
 userVariables []
}
