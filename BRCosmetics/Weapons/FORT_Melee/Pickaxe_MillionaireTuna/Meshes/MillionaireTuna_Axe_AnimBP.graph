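# Animation Blueprint for the MillionaireTuna pickaxe that adds secondary motion to its two-pivot rope.
# Bone-relative AnimDynamics with state-driven gravity simulates the rope, and a cached variant adds a time-based sine wave while moving.
# A second, world-space dynamics pass applies partially only in the front end.
# Additive bone rotations then pose the rope when grounded (delayed onset), moving left while jumping or falling, or moving right.
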AnimBlueprint /BRCosmetics/Weapons/FORT_Melee/Pickaxe_MillionaireTuna/Meshes/MillionaireTuna_Axe_AnimBP
target_skeleton=&/MillionaireTuna_Axe_Skeleton
graphs [
 AnimGraph {nodes [
   n1 = Use cached pose 'DynRope Cache'()
   n2 = Use cached pose 'SinCache'()
   n3 = Get bIsMoving2D()
   n4 = Get alpha_isFrontEnd() -> (alpha_isFrontEnd:double)
   n5 = Get GravityOverride()
   n6 = Anim Dynamics - Bone: pivot_rope_01!(
    ComponentPose=Local To Component(LocalPose=Blend Poses by bool(BlendPose_0=@n2, BlendPose_1=@n1, BlendTime_0=0.2, BlendTime_1=0.2, bActiveValue=@n3) {blendType=Linear})
    Alpha=@n4
    GravityOverride:Vector=@n5
    ExternalForce="0,0,0"
   ) -> (Pose) {
    angularDampingOverride=0.997
    angularSpringConstant=12000
    bAngularSpring=true
    bLinearSpring=true
    bOverrideAngularDamping=true
    bOverrideLinearDamping=true
    bUseGravityOverride=true
    gravityScale=0.35
    linearDampingOverride=0.7
    linearSpringConstant=150
    simulationSpace=World
    alphaBoolBlend {blendOption=Linear}
    boundBone {boneName=pivot_rope_01}
    componentAppliedLinearAccClamp=<2000 2000 2000>
    componentLinearAccScale=<0.1 0.1 0.1>
    componentLinearVelScale=<0.2 0.2 0.1>
    gravityOverride {z=-980}
    physicsBodyDefinitions [{
      boundBone {boneName=pivot_rope_01}
      boxExtents=<20 20 50>
      constraintSetup {angularLimitsMax=<0 0 80>, angularLimitsMin=<0 0 -80>, angularTarget=<1 0 0>, linearAxesMin=<-2 0 0>}
      localJointOffset=<-3.5 -3.5 0>
     }]
    relativeSpaceBone {boneName=b_AR_Root}
   }
   n7 = Anim Dynamics - Bone: pivot_rope_02!(ComponentPose=@n6, Alpha=@n4, GravityOverride:Vector=@n5, ExternalForce="0,0,0") -> (Pose) {
    angularDampingOverride=0.998
    angularSpringConstant=12000
    bAngularSpring=true
    bOverrideAngularDamping=true
    bOverrideLinearDamping=true
    bUseGravityOverride=true
    gravityScale=0.35
    linearDampingOverride=0.75
    linearSpringConstant=500
    simulationSpace=World
    alphaBoolBlend {blendOption=Linear}
    boundBone {boneName=pivot_rope_02}
    componentAppliedLinearAccClamp=<2000 2000 2000>
    componentLinearAccScale=<0.1 0.1 0.1>
    componentLinearVelScale=<0.2 0.2 0.1>
    gravityOverride {z=-980}
    physicsBodyDefinitions [{
      boundBone {boneName=pivot_rope_02}
      boxExtents=<20 20 50>
      constraintSetup {angularLimitsMax=<0 20 180>, angularLimitsMin=<0 -20 -180>, angularTarget=<1 0 0>}
      localJointOffset=<-2 0 0>
     }]
    relativeSpaceBone {boneName=b_AR_Root}
   }
   n8 = Get alpha_isOnGround() -> (alpha_isOnGround:double)
   n9 = "Transform (Modify) Bone - Bone: base_rope_01"(ComponentPose=@n7, Alpha=@n8, Translation="0,0,0", Rotation="-20,0,0", Scale="1,1,1") {
    rotationMode=BMM_Additive
    rotationSpace=BCS_BoneSpace
    translationMode=BMM_Additive
    translationSpace=BCS_BoneSpace
    alphaBoolBlend {blendOption=Linear}
    alphaScaleBiasClamp {bInterpResult=true, interpSpeedDecreasing=5, interpSpeedIncreasing=15}
    boneToModify {boneName=base_rope_01}
    rotation {pitch=-20}
   }
   n10 = "Transform (Modify) Bone - Bone: pivot_rope_02"(ComponentPose=@n9, Alpha=@n8, Translation="0,0,0", Rotation="0,15,0", Scale="1,1,1") {
    rotationMode=BMM_Additive
    rotationSpace=BCS_BoneSpace
    translationMode=BMM_Additive
    translationSpace=BCS_BoneSpace
    alphaBoolBlend {blendOption=Linear}
    alphaScaleBiasClamp {bInterpResult=true, interpSpeedDecreasing=5, interpSpeedIncreasing=15}
    boneToModify {boneName=pivot_rope_02}
    rotation {yaw=15}
   }
   n11 = Get Left() -> (Left:double)
   n12 = "Transform (Modify) Bone - Bone: pivot_rope_01"(ComponentPose=@n10, Alpha=@n11, Translation="0,0,0", Rotation="0,30,0", Scale="1,1,1") {
    rotationMode=BMM_Additive
    rotationSpace=BCS_BoneSpace
    translationMode=BMM_Additive
    translationSpace=BCS_BoneSpace
    alphaBoolBlend {blendOption=Linear}
    alphaScaleBiasClamp {bInterpResult=true, interpSpeedDecreasing=5, interpSpeedIncreasing=15}
    boneToModify {boneName=pivot_rope_01}
    rotation {yaw=30}
   }
   n13 = "Transform (Modify) Bone - Bone: pivot_rope_02"(ComponentPose=@n12, Alpha=@n11, Translation="0,0,0", Rotation="0,50,0", Scale="1,1,1") {
    rotationMode=BMM_Additive
    rotationSpace=BCS_BoneSpace
    translationMode=BMM_Additive
    translationSpace=BCS_BoneSpace
    alphaBoolBlend {blendOption=Linear}
    alphaScaleBiasClamp {bInterpResult=true, interpSpeedDecreasing=5, interpSpeedIncreasing=15}
    boneToModify {boneName=pivot_rope_02}
    rotation {yaw=50}
   }
   n14 = "Transform (Modify) Bone - Bone: base_rope_01"(ComponentPose=@n13, Alpha=Get Right() -> (Right:double), Translation="0,0,0", Rotation="5,0,0", Scale="1,1,1") {
    rotationMode=BMM_Additive
    rotationSpace=BCS_BoneSpace
    translationMode=BMM_Additive
    translationSpace=BCS_BoneSpace
    alphaBoolBlend {blendOption=Linear}
    alphaScaleBiasClamp {bInterpResult=true, interpSpeedDecreasing=5, interpSpeedIncreasing=15}
    boneToModify {boneName=base_rope_01}
    rotation {pitch=5}
   }
   Output Pose(Result=Component To Local(ComponentPose=@n14)) {layerGroup=None}
   AnimGraphNode_SubInput_0 = Input Pose()
   n15 = Get GravityOverride()
   n16 = Anim Dynamics - Bone: pivot_rope_01(ComponentPose=Local To Component(LocalPose=@AnimGraphNode_SubInput_0), Alpha=1, GravityOverride=@n15, ExternalForce="0,0,0") {
    angularDampingOverride=0.997
    angularSpringConstant=12000
    bAngularSpring=true
    bLinearSpring=true
    bOverrideAngularDamping=true
    bOverrideLinearDamping=true
    bUseGravityOverride=true
    gravityScale=0.35
    linearDampingOverride=0.7
    linearSpringConstant=150
    simulationSpace=BoneRelative
    alphaBoolBlend {blendOption=Linear}
    boundBone {boneName=pivot_rope_01}
    componentAppliedLinearAccClamp=<2000 2000 2000>
    componentLinearAccScale=<0.1 0.1 0.1>
    componentLinearVelScale=<0.2 0.2 0.1>
    gravityOverride {z=-980}
    physicsBodyDefinitions [{
      boundBone {boneName=pivot_rope_01}
      boxExtents=<20 20 50>
      constraintSetup {angularLimitsMax=<0 0 80>, angularLimitsMin=<0 0 -80>, angularTarget=<1 0 0>, linearAxesMin=<-2 0 0>}
      localJointOffset=<-3.5 -3.5 0>
     }]
    relativeSpaceBone {boneName=b_AR_Root}
   }
   n17 = Anim Dynamics - Bone: pivot_rope_02(ComponentPose=@n16, Alpha=1, GravityOverride=@n15, ExternalForce="0,0,0") {
    angularDampingOverride=0.998
    angularSpringConstant=12000
    bAngularSpring=true
    bOverrideAngularDamping=true
    bOverrideLinearDamping=true
    bUseGravityOverride=true
    gravityScale=0.35
    linearDampingOverride=0.75
    linearSpringConstant=500
    simulationSpace=BoneRelative
    alphaBoolBlend {blendOption=Linear}
    boundBone {boneName=pivot_rope_02}
    componentAppliedLinearAccClamp=<2000 2000 2000>
    componentLinearAccScale=<0.1 0.1 0.1>
    componentLinearVelScale=<0.2 0.2 0.1>
    gravityOverride {z=-980}
    physicsBodyDefinitions [{
      boundBone {boneName=pivot_rope_02}
      boxExtents=<20 20 50>
      constraintSetup {angularLimitsMax=<0 20 180>, angularLimitsMin=<0 -20 -180>, angularTarget=<1 0 0>}
      localJointOffset=<-2 0 0>
     }]
    relativeSpaceBone {boneName=b_AR_Root}
   }
   Save cached pose 'DynRope Cache'(Pose=Component To Local(ComponentPose=@n17)) {cacheName=DynRope Cache}
   n18 = "Transform (Modify) Bone - Bone: pivot_rope_01"(
    ComponentPose=Local To Component(LocalPose=Use cached pose 'DynRope Cache'())
    Alpha=1
    Translation="0,0,0"
    Rotation=Get rotatorWaveRope() -> (rotatorWaveRope:Rotator)
    Scale="1,1,1"
   ) {
    rotationMode=BMM_Additive
    rotationSpace=BCS_BoneSpace
    translationMode=BMM_Additive
    translationSpace=BCS_BoneSpace
    alphaBoolBlend {blendOption=Linear}
    alphaScaleBiasClamp {bInterpResult=true, interpSpeedDecreasing=5, interpSpeedIncreasing=15}
    boneToModify {boneName=pivot_rope_01}
    rotation {pitch=-15}
   }
   Save cached pose 'SinCache'(Pose=Component To Local(ComponentPose=@n18)) {cacheName=SinCache}
  ]}
 EventGraph {nodes [
   Event BlueprintUpdateAnimation()
   > n1 = Sequence!(execute) -> (then_0, then_1, then_2:Exec, then_3:Exec, then_4:Exec, then_5:Exec)
   > n2 = Gravity Override State Function DEPRECATED(CharacterPartAnimInstance=Self-Reference(), GameTimeInSeconds=GetGameTimeInSeconds(), GravityOverrideParams=Get Gravity Override Params())
   > Set GravityOverride(GravityOverride=@n2)
   n3 = Branch(execute=@n1.then_1, Condition=Get bIsOnGround() -> (bIsOnGround:bool))
   > Delay(Duration=0.17)
   > Set alpha_isOnGround(alpha_isOnGround:double=1) -> (Output_Get:double)
   Set alpha_isOnGround(execute=@n3.else, alpha_isOnGround=0)
   n4 = PlayerMovement(execute=@n1.then_2, AnimInstance=Self-Reference(), XDotProductCoefficient=100, YDotProductCoefficient=100, ZDotProductCoefficient=100)
   n5 = OR Boolean(A=Get bIsFalling(), B=Get bIsJumping())
   n6 = BreakVector(InVec=MakeVector(X=@n4, Y=@n4.YMovement))
   n7 = "float < float"(A=@n6, B=-50)
   > Set Left(Left:double="To Float (Boolean)"(InBool=AND Boolean(A=@n7, B=@n5))) -> (Output_Get:double)
   > Set Right(Right:double="To Float (Boolean)"(InBool="float > float"(A=@n6, B=50))) -> (Output_Get:double)
   n8 = Branch(execute=@n1.then_3, Condition=Get bIsMoving2D())
   > Set speedWaveRope(speedWaveRope:double=10) -> (Output_Get:double)
   > Set amplitudWaveRope(amplitudWaveRope:double=5) -> (Output_Get:double)
   > Set offsetWaveRope(offsetWaveRope:double=0) -> (Output_Get:double)
   Set speedWaveRope(execute=@n8.else, speedWaveRope=0)
   n9 = float * float(A=GetRealTimeSeconds(), B=Get speedWaveRope() -> (speedWaveRope:double)) {is=title K2Node_PromotableOperator}
   n10 = float * float(A=MultiplyByPi!(Value:double="Sin (Radians)"(A=@n9)) -> (ReturnValue:double), B=Get amplitudWaveRope() -> (amplitudWaveRope:double))
   n11 = Get offsetWaveRope() -> (offsetWaveRope:double)
   Set rotatorWaveRope(execute=@n1.then_4, rotatorWaveRope:Rotator=MakeRotator(Yaw=float + float(A=@n10, B=@n11))) -> (Output_Get:Rotator)
   n12 = Branch(execute=@n1.then_5, Condition=Get bIsInFrontEnd())
   > Set alpha_isFrontEnd(alpha_isFrontEnd:double=0.39) -> (Output_Get:double)
   Set alpha_isFrontEnd(execute=@n12.else, alpha_isFrontEnd=0)
  ]}
]
variables [
 alpha_isFrontEnd {type=number}
 alpha_isOnGround {type=number}
 amplitudWaveRope {type=number}
 Gravity Override Params {
  type=GravityOverrideParamsStruct
  default {
   emoteOrMelee {gravityOverride=<0 0 1000>}
   onGround_Falling {windFrequency=<1 0 0>}
   onGround_Moving {gravityOverride=<2500 2500 3000>}
   onGround_Standing {gravityOverride=<0 0 -1500>}
   skydive_GoingUp {gravityOverride=<0 0 1500>}
  }
 }
 GravityOverride {type=Vector}
 Left {type=number}
 offsetWaveRope {type=number}
 Right {type=number}
 rotatorWaveRope {type=Rotator}
 speedWaveRope {type=number}
]
